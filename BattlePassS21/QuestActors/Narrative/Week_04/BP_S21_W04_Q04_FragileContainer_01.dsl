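; Season 21 Week 4 Quest 4 fragile container, a battle pass narrative quest prop actor.
; On the post-interaction visual and audio event, it runs the parent behavior, then hides its quest icon component on the map, minimap and in the world.
; All other logic comes from the parent class, with begin play and construction only calling through and the overlap and tick events left disabled.
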(event VisualState|VFX&Audio|EventPostInteractionVFX&Audio
  (params (PlayerController :type Fort%20Player%20Controller%20Object%20Reference))
  (bind _questiconcomponent (Variables|BuildingGameplayActorPropQuest|GetQuestIconComponent))
  (|CallParentFunction 0)
  (MiniMapData|SetIconVisibilityonMap _questiconcomponent false)
  (MiniMapData|SetIconVisibilityonMiniMap _questiconcomponent false)
  (MiniMapData|SetIconVisibilityinWorld _questiconcomponent false))

(event EventBeginPlay
  (|CallParentFunction))

(event Collision|EventActorBeginOverlap
  (params (OtherActor :type Actor%20Object%20Reference))
  (|CallParentFunction OtherActor) ; This node is disabled and will not be called.\nDrag off pins to build functionality.
)

(event EventTick
  (params (DeltaSeconds :type Float%20%28single-precision%29))
  (|CallParentFunction DeltaSeconds) ; This node is disabled and will not be called.\nDrag off pins to build functionality.
)

(fn ConstructionScript
  (|CallParentFunction))
